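# Holiday-themed Niagara vehicle boost effect that streams Christmas balls, holly and pine sprites, and colored ribbons from skeletal mesh sockets.
# Ball meshes and leaf sprites spawn across the ports, fly backward and scatter with curl noise and drag.
# Green, red and white ribbons swirl through age-and-time sine/cosine offsets, with lifetime set by GameModeID.
# Preview flags slow ball and leaf speed; the lights and extra white-trail emitters are disabled.
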NiagaraSystem /VehicleCosmetics/Boosts/Boost_Holidaydeco/FX/NS_Boost_Holidaydeco
dependencies {
 usedDataInterfaces [
  &NiagaraDataInterfaceCurve
  &NiagaraDataInterfaceSpriteRendererInfo
  &NiagaraDataInterfaceSkeletalMesh
  &NiagaraDataInterfaceArrayFloat3
 ]
 usedDynamicInputs [&Vector2DFromFloat, &SelectVectorFromArray]
 usedModules [
  &self:SystemSpawnScript.SystemScriptSource.SystemScriptGraph.NiagaraNodeAssignment_0.SetVariables_#1
  &self:SystemSpawnScript.SystemScriptSource.SystemScriptGraph.NiagaraNodeAssignment_1.SetVariables_#2
  &SystemState
  &self:SystemSpawnScript.SystemScriptSource.SystemScriptGraph.NiagaraNodeAssignment_2.SetVariables_#3
 ]
}
emitters {
 NE_Cristmasballs {
  emitterUpdateScript [
   &EmitterState(Inactive Response=::NewEnumerator1, Life Cycle Mode=::NewEnumerator1, UseLoopDelay=false)
   &VC_FilteredSocketCount()
   &SpawnRate(SpawnRate=2, Use Spawn Probability=false)
   &SpawnRate(SpawnGroup=1, SpawnRate=&RandomRangeFloat(Minimum=2, Maximum=4), Use Spawn Probability=false) {module=SpawnRate002}
   &SpawnRate(SpawnGroup=2, SpawnRate=&RandomRangeFloat(Minimum=2, Maximum=4), Use Spawn Probability=false) {module=SpawnRate003}
   &SpawnRate(SpawnGroup=3, SpawnRate=&RandomRangeFloat(Minimum=2, Maximum=5), Use Spawn Probability=false) {module=SpawnRate004}
  ]
  particleSpawnScript [
   /SetVariables(Particles.VisibilityTag=$Emitter.SpawnGroup) {isSetParameters=true}
   &VC_PaintedVariantsOverride() {enabled=false}
   &InitializeParticle(
    Color Mode=::NewEnumerator0
    Lifetime=&RandomRangeFloat(Minimum=2, Maximum=2.5)
    Mesh Scale Mode=::NewEnumerator1
    Mesh Uniform Scale Max=0.1
    Mesh Uniform Scale Min=0.05
    Sprite Rotation Mode=::NewEnumerator2
    UsePositionOffset=false
    Write Material Random=true
   )
   &PartitionParticles(Number Of Partitions=$Emitter.PortCount)
   &SkeletalMeshLocation(
    Bone / Socket Index=$Particles.PartitionParticles.Partition
    Bone / Socket Sampling Mode=::NewEnumerator5
    MaskSpawnGroup=false
    Mesh Sampling Type=::NewEnumerator6
    Sampled Position Offset=&AddVectorToPosition(Position=<0 0 0>, Vector=<-50 0 0>)
    Skeletal Mesh=$System.SkeletalMesh
   )
   &AddVelocity(
    Cone Angle=10
    Module.UseConeFalloff=false
    Module.UseVelDistribution=false
    Override Seed=false
    Velocity Mode=::NewEnumerator2
    Velocity Speed=&MakeCustomFloatFromBool(Bool=$User.bIsFrontendPreview, True Float=-400, False Float=-1200)
   )
   &ShapeLocation(MaskSpawnGroup=false, Non Uniform Scale=<1 1500 1>, Override Seed=false, Sphere Radius=45, Sphere Surface Distribution=1, UseOwnerScale=true)
   &InitialMeshOrientation(Use Rotation Vector=false)
   /VehicleCosmetics/FX/VC_FXParents/Modules/DLR_ProductPreviews(GameModeID:NiagaraInt32=$System.GameModeID, bPreviewVehicleSpeed:NiagaraBool=$System.bPreviewVehicleSpeed)
  ]
  particleUpdateScript [
   &ParticleState()
   &ScaleMeshSize(Scale Factor=&VectorFromFloat(Value=&RandomRangeFloat(
      Minimum=&FloatFromCurve(FloatCurve=DataInterface() {propertyValues {Curve {
          defaultValue=1
          keys [{interpMode=RCIM_Cubic, tangentMode=RCTM_User, time=0.5, value=1}, {interpMode=RCIM_Cubic, tangentMode=RCTM_User, time=1, value=0}]
         }}})
      Maximum=&FloatFromCurve(FloatCurve=DataInterface() {propertyValues {Curve {
          defaultValue=1
          keys [
           {arriveTangent=9.0725, interpMode=RCIM_Cubic, leaveTangent=9.0725, tangentMode=RCTM_User, time=0, value=0}
           {arriveTangent=0.4749, interpMode=RCIM_Cubic, leaveTangent=0.4749, tangentMode=RCTM_User, time=0.1408, value=0.9847}
           {arriveTangent=-0.0405, interpMode=RCIM_Cubic, leaveTangent=-0.0405, tangentMode=RCTM_User, time=1, value=0}
          ]
         }}})
     )))
   &UpdateMeshOrientation(Rotation Rate=&RandomRangeFloat(Minimum=0.5, Maximum=1), Rotation Vector=<1 1 0>)
   &SpriteRotationRate(Rotation Rate=&RandomRangeFloat(Minimum=-120, Maximum=120))
   &CurlNoiseForce(
    Bias Noise=false
    Noise Frequency=&RandomRangeFloat(Minimum=25, Maximum=65)
    Noise Strength=1000
    Pan Noise=false
   )
   &Drag(Drag=&RandomRangeFloat(Minimum=0.5, Maximum=1.5), Use Linear Drag=true, Use Rotational Drag=true)
   &InheritSourceMovement(ApplyForce=false, ApplyPosition=true, ApplyPositionFalloff=false)
   /SetVariables(
    Particles.GreenBallPos:NiagaraPosition=&AddVectorToPosition(
     Position=$Particles.Position
     Vector=&Multiply_Vector(
      A=&VectorFromFloat(Value=&FloatFromCurve(
        FloatCurve=DataInterface() {propertyValues {Curve {defaultValue=1, keys [{tangentMode=RCTM_User, time=0, value=0}, {tangentMode=RCTM_User, time=1, value=1}]}}}
        Scale Curve=&RandomRangeFloat(Minimum=1, Maximum=2)
       ))
      B=&RandomRangeVector(Minimum=<-60 -60 -60>, Maximum=<60 60 60>)
     )
    )
    Particles.GreenLightPos:NiagaraPosition=&AddVectorToPosition(
     Position=$Particles.Position
     Vector=&Multiply_Vector(
      A=&VectorFromFloat(Value=&FloatFromCurve(
        FloatCurve=DataInterface() {propertyValues {Curve {defaultValue=1, keys [{tangentMode=RCTM_User, time=0, value=0}, {tangentMode=RCTM_User, time=1, value=1}]}}}
        Scale Curve=&RandomRangeFloat(Minimum=1, Maximum=2)
       ))
      B=&RandomRangeVector(Minimum=<-50 -50 -50>, Maximum=<50 50 50>)
     )
    )
    Particles.RedLightPos:NiagaraPosition=&AddVectorToPosition(
     Position=$Particles.Position
     Vector=&Multiply_Vector(
      A=&VectorFromFloat(Value=&FloatFromCurve(
        FloatCurve=DataInterface() {propertyValues {Curve {defaultValue=1, keys [{tangentMode=RCTM_User, time=0, value=0}, {tangentMode=RCTM_User, time=1, value=1}]}}}
        Scale Curve=&RandomRangeFloat(Minimum=1, Maximum=2)
       ))
      B=&RandomRangeVector(Minimum=<-45 -45 -45>, Maximum=<45 45 45>)
     )
    )
   ) {isSetParameters=true, module=SetVariables_#4}
   &InheritSourceMovement(
    Applied Position Scale=&Multiply_VectorByFloat(
     Vector=<1 1 1>
     Float=&FloatFromCurve(FloatCurve=DataInterface() {propertyValues {Curve {
         defaultValue=1
         keys [
          {arriveTangent=-1.7885, interpMode=RCIM_Cubic, leaveTangent=-1.7885, tangentMode=RCTM_User, time=0.05, value=1}
          {interpMode=RCIM_Cubic, tangentMode=RCTM_User, time=1, value=0}
         ]
        }}})
    )
    ApplyForce=false
    ApplyPosition=true
    ApplyPositionFalloff=false
   ) {enabled=false, module=InheritSourceMovement001}
   &DynamicMaterialParameters(Write Parameter Index 0=::NewEnumerator0)
   &SolveForcesAndVelocity(Clamp Velocity=false, Limit Acceleration=false)
  ]
  rendererClasses [&NiagaraMeshRendererProperties, &NiagaraSpriteRendererProperties]
  renderers [
   {rendererClass=&NiagaraMeshRendererProperties, rendererIndex=0}
   {rendererClass=&NiagaraMeshRendererProperties, rendererIndex=1}
   {rendererClass=&NiagaraMeshRendererProperties, rendererIndex=2}
   {rendererClass=&NiagaraMeshRendererProperties, rendererIndex=3}
   {rendererClass=&NiagaraSpriteRendererProperties, rendererIndex=4}
  ]
 }
 NE_Cristmaslights {
  bEnabled=false
  emitterUpdateScript [
   &EmitterState()
   &VC_FilteredSocketCount()
   &SpawnRate(SpawnRate=&RandomRangeFloat(Minimum=3, Maximum=5), Use Spawn Probability=false)
  ]
  particleSpawnScript [
   &VC_PaintedVariantsOverride() {enabled=false}
   &InitializeParticle(
    Color Mode=::NewEnumerator0
    Lifetime=&RandomRangeFloat(Minimum=2, Maximum=2.5)
    Mesh Scale Mode=::NewEnumerator1
    Mesh Uniform Scale Max=0.2
    Mesh Uniform Scale Min=0.05
    Sprite Rotation Mode=::NewEnumerator2
    UsePositionOffset=false
    Write Material Random=true
   )
   &PartitionParticles(Number Of Partitions=$Emitter.PortCount)
   &SkeletalMeshLocation(
    Bone / Socket Index=$Particles.PartitionParticles.Partition
    Bone / Socket Sampling Mode=::NewEnumerator5
    MaskSpawnGroup=false
    Mesh Sampling Type=::NewEnumerator6
    Sampled Position Offset=&AddVectorToPosition(Position=<0 0 0>, Vector=<-50 0 0>)
    Skeletal Mesh=$System.SkeletalMesh
   )
   &AddVelocity(Velocity=&RandomRangeVector(Minimum=<-800 0 0>, Maximum=<-1000 1 1>))
   &ShapeLocation(Cone Angle=180, Cone Inner Angle=180, MaskSpawnGroup=false, Override Seed=false, Shape Primitive=::NewEnumerator5, UseOwnerScale=true)
   &InitialMeshOrientation(Use Rotation Vector=false)
  ]
  particleUpdateScript [
   &ParticleState()
   &ScaleMeshSize(Scale Factor=&VectorFromFloat(Value=&RandomRangeFloat(
      Minimum=&FloatFromCurve(FloatCurve=DataInterface() {propertyValues {Curve {
          defaultValue=1
          keys [{interpMode=RCIM_Cubic, tangentMode=RCTM_User, time=0.5, value=1}, {interpMode=RCIM_Cubic, tangentMode=RCTM_User, time=1, value=0}]
         }}})
      Maximum=&FloatFromCurve(FloatCurve=DataInterface() {propertyValues {Curve {
          defaultValue=1
          keys [
           {arriveTangent=9.0725, interpMode=RCIM_Cubic, leaveTangent=9.0725, tangentMode=RCTM_User, time=0, value=0}
           {arriveTangent=0.4749, interpMode=RCIM_Cubic, leaveTangent=0.4749, tangentMode=RCTM_User, time=0.1408, value=0.9847}
           {arriveTangent=-0.0405, interpMode=RCIM_Cubic, leaveTangent=-0.0405, tangentMode=RCTM_User, time=1, value=0}
          ]
         }}})
     )))
   &UpdateMeshOrientation(Rotation Rate=&RandomRangeFloat(Minimum=0.5, Maximum=1), Rotation Vector=<1 1 0>)
   &SpriteRotationRate(Rotation Rate=&RandomRangeFloat(Minimum=-120, Maximum=120))
   &CurlNoiseForce(
    Bias Noise=false
    Noise Frequency=&RandomRangeFloat(Minimum=25, Maximum=65)
    Noise Strength=1000
    Pan Noise=false
   )
   &Drag(Drag=&RandomRangeFloat(Minimum=0.5, Maximum=1.5), Use Linear Drag=true, Use Rotational Drag=true)
   &InheritSourceMovement(ApplyForce=false, ApplyPosition=true, ApplyPositionFalloff=false)
   &SolveForcesAndVelocity(Clamp Velocity=false, Limit Acceleration=false)
   &DynamicMaterialParameters(
    Index 0 Param 1=30
    Module.Index0Float0Write=true
    Module.Index0Float1Write=false
    Module.Index0Float2Write=false
    Module.Index0Float3Write=false
   ) {enabled=false}
   /SetVariables(Particles.GreenPos:NiagaraPosition=&AddVectorToPosition(
     Position=$Particles.Position
     Vector=&Multiply_Vector(
      A=&VectorFromFloat(Value=&FloatFromCurve(
        FloatCurve=DataInterface() {propertyValues {Curve {defaultValue=1, keys [{tangentMode=RCTM_User, time=0, value=0}, {tangentMode=RCTM_User, time=1, value=1}]}}}
        Scale Curve=&RandomRangeFloat(Minimum=1, Maximum=2)
       ))
      B=&RandomRangeVector(Minimum=<-150 -150 -150>, Maximum=<150 150 150>)
     )
    )) {isSetParameters=true, module=SetVariables_#4}
  ]
  rendererClasses [&NiagaraMeshRendererProperties, &NiagaraSpriteRendererProperties]
  renderers [
   {rendererClass=&NiagaraMeshRendererProperties, rendererIndex=0}
   {rendererClass=&NiagaraMeshRendererProperties, rendererIndex=1}
   {rendererClass=&NiagaraSpriteRendererProperties, rendererIndex=2}
  ]
 }
 NE_Green_Trail {
  emitterUpdateScript [
   &EmitterState(Life Cycle Mode=::NewEnumerator1, Loop Duration=1, UseLoopDelay=false)
   &VC_FilteredSocketCount()
   &VC_RateQualityHandler()
   &VC_SpawnRibbon_v1(Number of Ribbons=$Emitter.PortCount, SpawnRate=$Emitter.Spawn Rate)
   /SetVariables(Emitter.TimeOffset:NiagaraFloat=&Power(a=HlslExpression() {hlslExpression="(fmod(Engine.Time * 1.0, 0.3) + 0.15) * 3"}, b=1)) {isSetParameters=true, module=SetVariables_#5}
  ]
  particleSpawnScript [
   &VC_PaintedVariantsOverride() {enabled=false}
   &VC_InitializeRibbon_v1(
    Color=<rgba 0.0529 0.1531 0.0196 0.5>
    Custom Offset=true
    Enable Skeletal Position=true
    Facing Axis:ENiagaraOrientationAxis=::YAxis
    Lifetime=&MergeHelper_FloatSwitch(0=1.5, 1=0.1, 2=3, 3=?, Index=$User.GameModeID, Num=::NewEnumerator1)
    Offset Amount=1
    Ribbon Width=20
   )
   &AddVelocity(Velocity=&MakeCustomVectorFromBool(Bool=true, True Vector=<-1300 0 0>, False Vector=<-1200 0 0>)) {module=AddVelocity001}
   &ShapeLocation(MaskSpawnGroup=false, Override Seed=false, Sphere Radius=2, UseOwnerScale=true) {enabled=false}
   /SetVariables(Particles.BlackRibbon:LinearColor=<rgba 0 0 0 1>) {isSetParameters=true, module=SetVariables_#6}
   &DLR_ProductPreviews(GameModeID:NiagaraInt32=$System.GameModeID, bPreviewVehicleSpeed:NiagaraBool=$System.bPreviewVehicleSpeed)
  ]
  particleUpdateScript [
   &ParticleState()
   &ScaleRibbonWidth(Ribbon Width Scale=&FloatFromCurve(FloatCurve=DataInterface() {propertyValues {Curve {keys [
         {arriveTangent=-0.375, interpMode=RCIM_Cubic, time=0, value=1}
         {arriveTangent=-1.7516, interpMode=RCIM_Cubic, leaveTangent=-1.7516, tangentMode=RCTM_User, time=0.8, value=0}
        ]}}}))
   /SetVariables(Particles.Position=&AddVectorToPosition(
     Position=$Particles.Position
     Vector=&TransformVector(
      Vector=&MakeVector(
       X=0
       Y=&Cosine(
        Normalized Angle=&Add_Float(
         A=&Multiply_Float(A=$Particles.NormalizedAge, B=2)
         B=&Multiply_Float(A=$Emitter.LoopedAge, B=-2)
        )
        Period=1
        Scale=&FloatFromCurve(
         FloatCurve=DataInterface() {propertyValues {Curve {
            defaultValue=1
            keys [
             {arriveTangent=18.6243, interpMode=RCIM_Cubic, leaveTangent=18.6243, tangentMode=RCTM_User, time=0, value=0}
             {arriveTangent=-11.3665, interpMode=RCIM_Cubic, leaveTangent=-11.3665, tangentMode=RCTM_User, time=0.05, value=3}
             {arriveTangent=0.0224, interpMode=RCIM_Cubic, leaveTangent=0.0224, tangentMode=RCTM_User, time=0.3, value=-1}
            ]
           }}}
         Scale Curve=3
        )
        Bias=0
       )
       Z=&Sine(
        Normalized Angle=&Add_Float(
         A=&Multiply_Float(A=$Particles.NormalizedAge, B=2)
         B=&Multiply_Float(A=$Emitter.LoopedAge, B=-2)
        )
        Period=1
        Scale=&FloatFromCurve(
         FloatCurve=DataInterface() {propertyValues {Curve {
            defaultValue=1
            keys [
             {arriveTangent=18.6243, interpMode=RCIM_Cubic, leaveTangent=18.6243, tangentMode=RCTM_User, time=0, value=0}
             {arriveTangent=-11.3665, interpMode=RCIM_Cubic, leaveTangent=-11.3665, tangentMode=RCTM_User, time=0.05, value=3}
             {arriveTangent=0.0224, interpMode=RCIM_Cubic, leaveTangent=0.0224, tangentMode=RCTM_User, time=0.3, value=-1}
            ]
           }}}
         Scale Curve=2
        )
        Bias=0
       )
      )
      Destination Space=::Simulation
      Apply Scale=true
     )
    )) {isSetParameters=true, module=SetVariables_#7}
   &InheritSourceMovement(
    Applied Position Scale=&VectorFromFloat(Value=&FloatFromCurve(FloatCurve=DataInterface() {propertyValues {Curve {
         defaultValue=1
         keys [
          {arriveTangent=-1.7885, interpMode=RCIM_Cubic, leaveTangent=-1.7885, tangentMode=RCTM_User, time=0.05, value=1}
          {interpMode=RCIM_Cubic, tangentMode=RCTM_User, time=1, value=0}
         ]
        }}}))
    ApplyForce=false
    ApplyPosition=true
    ApplyPositionFalloff=false
   )
   &SolveForcesAndVelocity(Clamp Velocity=false, Limit Acceleration=false)
   &ScaleColor(
    Scale Alpha=&FloatFromCurve(FloatCurve=DataInterface() {propertyValues {Curve {keys [<at 0 1>,<at 0.2 0.4>,<at 1 0.3>]}}})
    ScaleA=true
    ScaleRGB=true
   )
   &DynamicMaterialParameters(Write Parameter Index 0=::NewEnumerator0)
   &VC_UpdateRibbon_v1(Enable Skeletal Position=true)
  ]
  rendererClasses [&NiagaraRibbonRendererProperties]
  renderers [
   {rendererClass=&NiagaraRibbonRendererProperties}
   {rendererClass=&NiagaraRibbonRendererProperties, rendererIndex=1}
  ]
 }
 NE_Hollyleaves {
  emitterUpdateScript [
   &EmitterState(Inactive Response=::NewEnumerator1, Life Cycle Mode=::NewEnumerator1, UseLoopDelay=false)
   &VC_FilteredSocketCount()
   &SpawnRate(SpawnRate=&RandomRangeFloat(Minimum=15, Maximum=20), Use Spawn Probability=false)
   &SpawnRate(SpawnGroup=1, SpawnRate=5, Use Spawn Probability=false) {module=SpawnRate002}
  ]
  particleSpawnScript [
   &VC_PaintedVariantsOverride() {enabled=false}
   &InitializeParticle(
    Color Mode=::NewEnumerator0
    Sprite Rotation Mode=::NewEnumerator2
    Sprite Size Mode=::NewEnumerator1
    Uniform Sprite Size Max=50
    Uniform Sprite Size Min=15
    UsePositionOffset=false
    Write Material Random=true
   )
   &PartitionParticles(Number Of Partitions=$Emitter.PortCount)
   &SkeletalMeshLocation(
    Bone / Socket Index=$Particles.PartitionParticles.Partition
    Bone / Socket Sampling Mode=::NewEnumerator5
    MaskSpawnGroup=false
    Mesh Sampling Type=::NewEnumerator6
    Sampled Position Offset=&AddVectorToPosition(Position=<0 0 0>, Vector=<-50 0 0>)
    Skeletal Mesh=$System.SkeletalMesh
   )
   &AddVelocity(
    Cone Angle=15
    Module.UseConeFalloff=false
    Module.UseVelDistribution=false
    Override Seed=false
    Velocity Mode=::NewEnumerator2
    Velocity Speed=&MakeCustomFloatFromBool(Bool=$User.bPreviewVehicleSpeed, True Float=-400, False Float=-1200)
   )
   &ShapeLocation(MaskSpawnGroup=false, Override Seed=false, Sphere Radius=45, UseOwnerScale=true)
   /SetVariables(Particles.VisibilityTag=$Emitter.SpawnGroup) {isSetParameters=true}
   &DLR_ProductPreviews(GameModeID:NiagaraInt32=$System.GameModeID, bPreviewVehicleSpeed:NiagaraBool=$System.bPreviewVehicleSpeed)
  ]
  particleUpdateScript [
   &ParticleState()
   &ScaleSpriteSize(Uniform Curve Sprite Scale=DataInterface() {propertyValues {Curve {
       defaultValue=0
       keys [
        {arriveTangent=0.1613, interpMode=RCIM_Cubic, leaveTangent=0.1613, tangentMode=RCTM_User, time=0.25, value=1}
        {arriveTangent=0.1613, interpMode=RCIM_Cubic, leaveTangent=0.1613, tangentMode=RCTM_User, time=1, value=0.25}
       ]
      }}})
   &SpriteRotationRate(Rotation Rate=&RandomRangeFloat(Minimum=-120, Maximum=120))
   &ScaleColor(
    Scale Alpha=&FloatFromCurve(FloatCurve=DataInterface() {propertyValues {Curve {keys [<at 0 1>,<at 0.9 0 RCIM_Cubic>]}}})
    ScaleA=true
    ScaleRGB=true
   ) {enabled=false}
   &CurlNoiseForce(
    Bias Noise=false
    Noise Frequency=&RandomRangeFloat(Minimum=25, Maximum=65)
    Noise Strength=1000
    Pan Noise=false
   )
   &Drag(Drag=&RandomRangeFloat(Minimum=0.5, Maximum=1.5), Use Linear Drag=true, Use Rotational Drag=true)
   &InheritSourceMovement(ApplyForce=false, ApplyPosition=true, ApplyPositionFalloff=false)
   &InheritSourceMovement(
    Applied Position Scale=&Multiply_VectorByFloat(
     Vector=<1 1 1>
     Float=&FloatFromCurve(FloatCurve=DataInterface() {propertyValues {Curve {
         defaultValue=1
         keys [
          {arriveTangent=-1.7885, interpMode=RCIM_Cubic, leaveTangent=-1.7885, tangentMode=RCTM_User, time=0.05, value=1}
          {interpMode=RCIM_Cubic, tangentMode=RCTM_User, time=1, value=0}
         ]
        }}})
    )
    ApplyForce=false
    ApplyPosition=true
    ApplyPositionFalloff=false
   ) {enabled=false, module=InheritSourceMovement001}
   &DynamicMaterialParameters(
    Index 0 Param 1=50
    Module.Index0Float0Write=true
    Module.Index0Float1Write=false
    Module.Index0Float2Write=false
    Module.Index0Float3Write=false
   )
   &SubUVAnimation(
    Random Start Frame=true
    "Sprite Renderer "=DataInterface() {propertyValues {SpriteRenderer=&self:NE_Hollyleaves_0.NiagaraSpriteRendererProperties_6}}
    UseEndFrame=false
    UseStartFrame=false
   )
   /SetVariables(
    Particles.LeavesColor:LinearColor=<rgba 0.3497 0.9865 0 1>
    Particles.PineColor:LinearColor=<rgba 2 0.6583 0 1>
    Particles.PinesPosition:NiagaraPosition=&AddVectorToPosition(
     Position=$Particles.Position
     Vector=&Multiply_Vector(
      A=&VectorFromFloat(Value=&FloatFromCurve(
        FloatCurve=DataInterface() {propertyValues {Curve {defaultValue=1, keys [{tangentMode=RCTM_User, time=0, value=0}, {tangentMode=RCTM_User, time=1, value=1}]}}}
        Scale Curve=&RandomRangeFloat(Minimum=1, Maximum=2)
       ))
      B=&RandomRangeVector(Minimum=<-80 -80 -80>, Maximum=<80 80 80>)
     )
    )
    Particles.RotationPines:NiagaraFloat=&RandomRangeFloat(Minimum=-90, Maximum=90)
   ) {isSetParameters=true}
   &ScaleColor(
    Scale Alpha=&FloatFromCurve(FloatCurve=DataInterface() {propertyValues {Curve {keys [<at 0 1>,<at 1 0>]}}})
    Scale RGB=<5 5 5>
    ScaleA=true
    ScaleRGB=true
   ) {module=ScaleColor001}
   &SolveForcesAndVelocity(Clamp Velocity=false, Limit Acceleration=false)
  ]
  rendererClasses [&NiagaraSpriteRendererProperties]
  renderers [
   {rendererClass=&NiagaraSpriteRendererProperties}
   {rendererClass=&NiagaraSpriteRendererProperties, rendererIndex=1}
   {rendererClass=&NiagaraSpriteRendererProperties, rendererIndex=2}
  ]
 }
 NE_Red_Trail002 {
  emitterUpdateScript [
   &EmitterState(Life Cycle Mode=::NewEnumerator1, Loop Duration=1, UseLoopDelay=false)
   &VC_FilteredSocketCount()
   &VC_RateQualityHandler()
   &VC_SpawnRibbon_v1(Number of Ribbons=$Emitter.PortCount, SpawnRate=$Emitter.Spawn Rate)
   /SetVariables(Emitter.TimeOffset:NiagaraFloat=&Power(a=HlslExpression() {hlslExpression="(fmod(Engine.Time * 1.0, 0.3) + 0.15) * 3"}, b=1)) {isSetParameters=true, module=SetVariables_#5}
  ]
  particleSpawnScript [
   &VC_PaintedVariantsOverride() {enabled=false}
   &VC_InitializeRibbon_v1(
    Color=<rgba 0.3374 0 0 1>
    Custom Offset=true
    Enable Skeletal Position=true
    Facing Axis:ENiagaraOrientationAxis=::YAxis
    Lifetime=&MergeHelper_FloatSwitch(0=1.5, 1=0.1, 2=3, 3=?, Index=$User.GameModeID, Num=::NewEnumerator1)
    Offset Amount=1
    Ribbon Width=20
   )
   &AddVelocity(Velocity=&MakeCustomVectorFromBool(Bool=true, True Vector=<-1300 0 0>, False Vector=<-1200 0 0>)) {module=AddVelocity001}
   &ShapeLocation(MaskSpawnGroup=false, Override Seed=false, Sphere Radius=2, UseOwnerScale=true) {enabled=false}
   /SetVariables(Particles.BlackRibbon:LinearColor=<rgba 0 0 0 1>) {isSetParameters=true, module=SetVariables_#6}
   &DLR_ProductPreviews(GameModeID:NiagaraInt32=$System.GameModeID, bPreviewVehicleSpeed:NiagaraBool=$System.bPreviewVehicleSpeed)
  ]
  particleUpdateScript [
   &ParticleState()
   &ScaleRibbonWidth(Ribbon Width Scale=&FloatFromCurve(FloatCurve=DataInterface() {propertyValues {Curve {keys [
         {arriveTangent=-0.375, interpMode=RCIM_Cubic, time=0, value=1}
         {arriveTangent=-1.7516, interpMode=RCIM_Cubic, leaveTangent=-1.7516, tangentMode=RCTM_User, time=0.8, value=0}
        ]}}}))
   /SetVariables(Particles.Position=&AddVectorToPosition(
     Position=$Particles.Position
     Vector=&TransformVector(
      Vector=&MakeVector(
       X=0
       Y=&Cosine(
        Normalized Angle=&Add_Float(
         A=&Multiply_Float(A=$Particles.NormalizedAge, B=2)
         B=&Multiply_Float(A=$Emitter.LoopedAge, B=-3)
        )
        Period=1
        Scale=&FloatFromCurve(FloatCurve=DataInterface() {propertyValues {Curve {
            defaultValue=1
            keys [
             {arriveTangent=18.6243, interpMode=RCIM_Cubic, leaveTangent=18.6243, tangentMode=RCTM_User, time=0, value=0}
             {arriveTangent=-11.3665, interpMode=RCIM_Cubic, leaveTangent=-11.3665, tangentMode=RCTM_User, time=0.05, value=3}
             {arriveTangent=0.0224, interpMode=RCIM_Cubic, leaveTangent=0.0224, tangentMode=RCTM_User, time=0.3, value=-1}
            ]
           }}})
        Bias=0
       )
       Z=&Sine(
        Normalized Angle=&Add_Float(
         A=&Multiply_Float(A=$Particles.NormalizedAge, B=2)
         B=&Multiply_Float(A=$Emitter.LoopedAge, B=-3)
        )
        Period=1
        Scale=&FloatFromCurve(FloatCurve=DataInterface() {propertyValues {Curve {
            defaultValue=1
            keys [
             {arriveTangent=18.6243, interpMode=RCIM_Cubic, leaveTangent=18.6243, tangentMode=RCTM_User, time=0, value=0}
             {arriveTangent=-11.3665, interpMode=RCIM_Cubic, leaveTangent=-11.3665, tangentMode=RCTM_User, time=0.05, value=3}
             {arriveTangent=0.0224, interpMode=RCIM_Cubic, leaveTangent=0.0224, tangentMode=RCTM_User, time=0.3, value=-1}
            ]
           }}})
        Bias=0
       )
      )
      Destination Space=::Simulation
      Apply Scale=true
     )
    )) {isSetParameters=true, module=SetVariables_#7}
   &InheritSourceMovement(
    Applied Position Scale=&VectorFromFloat(Value=&FloatFromCurve(FloatCurve=DataInterface() {propertyValues {Curve {
         defaultValue=1
         keys [
          {arriveTangent=-1.7885, interpMode=RCIM_Cubic, leaveTangent=-1.7885, tangentMode=RCTM_User, time=0.05, value=1}
          {interpMode=RCIM_Cubic, tangentMode=RCTM_User, time=1, value=0}
         ]
        }}}))
    ApplyForce=false
    ApplyPosition=true
    ApplyPositionFalloff=false
   )
   &SolveForcesAndVelocity(Clamp Velocity=false, Limit Acceleration=false)
   &ScaleColor(
    Scale Alpha=&FloatFromCurve(FloatCurve=DataInterface() {propertyValues {Curve {keys [<at 0 1>,<at 0.2 0.4>,<at 1 0.3>]}}})
    ScaleA=true
    ScaleRGB=true
   )
   &DynamicMaterialParameters(Write Parameter Index 0=::NewEnumerator0)
   &VC_UpdateRibbon_v1(Enable Skeletal Position=true)
  ]
  rendererClasses [&NiagaraRibbonRendererProperties]
  renderers [
   {rendererClass=&NiagaraRibbonRendererProperties}
   {rendererClass=&NiagaraRibbonRendererProperties, rendererIndex=1}
  ]
 }
 NE_White_Trail {
  emitterUpdateScript [
   &EmitterState(Life Cycle Mode=::NewEnumerator1, Loop Duration=1, UseLoopDelay=false)
   &VC_FilteredSocketCount()
   &VC_RateQualityHandler()
   &VC_SpawnRibbon_v1(Number of Ribbons=$Emitter.PortCount, SpawnRate=$Emitter.Spawn Rate)
   /SetVariables(Emitter.TimeOffset:NiagaraFloat=&Power(a=HlslExpression() {hlslExpression="(fmod(Engine.Time * 1.0, 0.3) + 0.15) * 3"}, b=1)) {isSetParameters=true, module=SetVariables_#5}
  ]
  particleSpawnScript [
   &VC_PaintedVariantsOverride() {enabled=false}
   &VC_InitializeRibbon_v1(
    Color=<rgba 1 1 1 0.5>
    Custom Offset=true
    Enable Skeletal Position=true
    Facing Axis:ENiagaraOrientationAxis=::YAxis
    Lifetime=&MergeHelper_FloatSwitch(0=1.5, 1=0.1, 2=3, 3=?, Index=$User.GameModeID, Num=::NewEnumerator1)
    Offset Amount=1
    Ribbon Width=20
   )
   &AddVelocity(Velocity=&MakeCustomVectorFromBool(Bool=true, True Vector=<-1300 0 0>, False Vector=<-1200 0 0>)) {module=AddVelocity001}
   &ShapeLocation(MaskSpawnGroup=false, Override Seed=false, Sphere Radius=2, UseOwnerScale=true) {enabled=false}
   /SetVariables(Particles.BlackRibbon:LinearColor=<rgba 0 0 0 1>) {isSetParameters=true, module=SetVariables_#6}
   &DLR_ProductPreviews(GameModeID:NiagaraInt32=$System.GameModeID, bPreviewVehicleSpeed:NiagaraBool=$System.bPreviewVehicleSpeed)
  ]
  particleUpdateScript [
   &ParticleState()
   &ScaleRibbonWidth(Ribbon Width Scale=&FloatFromCurve(FloatCurve=DataInterface() {propertyValues {Curve {keys [
         {arriveTangent=-0.375, interpMode=RCIM_Cubic, time=0, value=1}
         {arriveTangent=-1.7516, interpMode=RCIM_Cubic, leaveTangent=-1.7516, tangentMode=RCTM_User, time=0.8, value=0}
        ]}}}))
   /SetVariables(Particles.Position=&AddVectorToPosition(
     Position=$Particles.Position
     Vector=&TransformVector(
      Vector=&MakeVector(
       X=0
       Y=&Cosine(
        Normalized Angle=&Add_Float(
         A=&Multiply_Float(A=$Particles.NormalizedAge, B=2)
         B=&Multiply_Float(A=$Emitter.LoopedAge, B=-2)
        )
        Period=1
        Scale=&FloatFromCurve(
         FloatCurve=DataInterface() {propertyValues {Curve {
            defaultValue=1
            keys [
             {arriveTangent=18.6243, interpMode=RCIM_Cubic, leaveTangent=18.6243, tangentMode=RCTM_User, time=0, value=0}
             {arriveTangent=-11.3665, interpMode=RCIM_Cubic, leaveTangent=-11.3665, tangentMode=RCTM_User, time=0.05, value=3}
             {arriveTangent=0.0224, interpMode=RCIM_Cubic, leaveTangent=0.0224, tangentMode=RCTM_User, time=0.3, value=-1}
            ]
           }}}
         Scale Curve=1.8
        )
        Bias=0
       )
       Z=&Sine(
        Normalized Angle=&Add_Float(
         A=&Multiply_Float(A=$Particles.NormalizedAge, B=2)
         B=&Multiply_Float(A=$Emitter.LoopedAge, B=-2)
        )
        Period=1
        Scale=&FloatFromCurve(
         FloatCurve=DataInterface() {propertyValues {Curve {
            defaultValue=1
            keys [
             {arriveTangent=18.6243, interpMode=RCIM_Cubic, leaveTangent=18.6243, tangentMode=RCTM_User, time=0, value=0}
             {arriveTangent=-11.3665, interpMode=RCIM_Cubic, leaveTangent=-11.3665, tangentMode=RCTM_User, time=0.05, value=3}
             {arriveTangent=0.0224, interpMode=RCIM_Cubic, leaveTangent=0.0224, tangentMode=RCTM_User, time=0.3, value=-1}
            ]
           }}}
         Scale Curve=0.8
        )
        Bias=0
       )
      )
      Destination Space=::Simulation
      Apply Scale=true
     )
    )) {isSetParameters=true, module=SetVariables_#7}
   &InheritSourceMovement(
    Applied Position Scale=&VectorFromFloat(Value=&FloatFromCurve(FloatCurve=DataInterface() {propertyValues {Curve {
         defaultValue=1
         keys [
          {arriveTangent=-1.7885, interpMode=RCIM_Cubic, leaveTangent=-1.7885, tangentMode=RCTM_User, time=0.05, value=1}
          {interpMode=RCIM_Cubic, tangentMode=RCTM_User, time=1, value=0}
         ]
        }}}))
    ApplyForce=false
    ApplyPosition=true
    ApplyPositionFalloff=false
   )
   &SolveForcesAndVelocity(Clamp Velocity=false, Limit Acceleration=false)
   &ScaleColor(
    Scale Alpha=&FloatFromCurve(FloatCurve=DataInterface() {propertyValues {Curve {keys [<at 0 1>,<at 0.2 0.4>,<at 1 0.3>]}}})
    ScaleA=true
    ScaleRGB=true
   )
   &DynamicMaterialParameters(Write Parameter Index 0=::NewEnumerator0)
   &VC_UpdateRibbon_v1(Enable Skeletal Position=true)
  ]
  rendererClasses [&NiagaraRibbonRendererProperties]
  renderers [
   {rendererClass=&NiagaraRibbonRendererProperties}
   {rendererClass=&NiagaraRibbonRendererProperties, rendererIndex=1}
  ]
 }
 NE_White_Trail001 {
  bEnabled=false
  emitterUpdateScript [
   &EmitterState(Life Cycle Mode=::NewEnumerator1, Loop Duration=1, UseLoopDelay=false)
   &VC_FilteredSocketCount()
   &VC_RateQualityHandler()
   &VC_SpawnRibbon_v1(Number of Ribbons=$Emitter.PortCount, SpawnRate=$Emitter.Spawn Rate)
   /SetVariables(Emitter.TimeOffset:NiagaraFloat=&Power(a=HlslExpression() {hlslExpression="(fmod(Engine.Time * 1.0, 0.3) + 0.15) * 3"}, b=1)) {isSetParameters=true, module=SetVariables_#5}
  ]
  particleSpawnScript [
   &VC_PaintedVariantsOverride() {enabled=false}
   &VC_InitializeRibbon_v1(Color=<rgba 1 1 1 0.6035>, Custom Offset=true, Enable Skeletal Position=true, Lifetime=2, Offset Amount=1, Ribbon Width=20)
   &AddVelocity(Velocity=&MakeCustomVectorFromBool(Bool=true, True Vector=<-1300 0 0>, False Vector=<-1200 0 0>)) {module=AddVelocity001}
   &ShapeLocation(MaskSpawnGroup=false, Override Seed=false, Sphere Radius=2, UseOwnerScale=true) {enabled=false}
   /SetVariables(Particles.BlackRibbon:LinearColor=<rgba 0 0 0 1>) {isSetParameters=true}
   &DLR_ProductPreviews(GameModeID:NiagaraInt32=$System.GameModeID, bPreviewVehicleSpeed:NiagaraBool=$System.bPreviewVehicleSpeed)
  ]
  particleUpdateScript [
   &ParticleState()
   &ScaleRibbonWidth(Ribbon Width Scale=&FloatFromCurve(FloatCurve=DataInterface() {propertyValues {Curve {keys [
         {arriveTangent=-0.0723, interpMode=RCIM_Cubic, leaveTangent=-0.0723, tangentMode=RCTM_User, time=0, value=1}
         {arriveTangent=-1.7516, interpMode=RCIM_Cubic, leaveTangent=-1.7516, tangentMode=RCTM_User, time=0.8, value=0}
        ]}}}))
   &InheritSourceMovement(
    Applied Position Scale=&VectorFromFloat(Value=&FloatFromCurve(FloatCurve=DataInterface() {propertyValues {Curve {
         defaultValue=1
         keys [
          {arriveTangent=-1.7885, interpMode=RCIM_Cubic, leaveTangent=-1.7885, tangentMode=RCTM_User, time=0.05, value=1}
          {interpMode=RCIM_Cubic, tangentMode=RCTM_User, time=1, value=0}
         ]
        }}}))
    ApplyForce=false
    ApplyPosition=true
    ApplyPositionFalloff=false
   ) {enabled=false}
   &SolveForcesAndVelocity(Clamp Velocity=false, Limit Acceleration=false)
   /SetVariables(Particles.Position=&AddVectorToPosition(
     Position=$Particles.Position
     Vector=&TransformVector(
      Vector=&MakeVector(
       X=0
       Y=&Sine(
        Normalized Angle=&Add_Float(
         A=&Multiply_Float(A=$Particles.NormalizedAge, B=2)
         B=&Multiply_Float(A=$Emitter.LoopedAge, B=-1)
        )
        Period=1
        Scale=&FloatFromCurve(
         FloatCurve=DataInterface() {propertyValues {Curve {
            defaultValue=1
            keys [
             {arriveTangent=18.6243, interpMode=RCIM_Cubic, leaveTangent=18.6243, tangentMode=RCTM_User, time=0, value=0}
             {arriveTangent=-11.3665, interpMode=RCIM_Cubic, leaveTangent=-11.3665, tangentMode=RCTM_User, time=0.1, value=4}
             {arriveTangent=0.0224, interpMode=RCIM_Cubic, leaveTangent=0.0224, tangentMode=RCTM_User, time=1, value=-2}
            ]
           }}}
         Scale Curve=0.1
        )
        Bias=0
       )
       Z=&Sine(
        Normalized Angle=&Add_Float(
         A=&Multiply_Float(A=$Particles.NormalizedAge, B=2)
         B=&Multiply_Float(A=$Emitter.LoopedAge, B=-1.5)
        )
        Period=2
        Scale=&FloatFromCurve(FloatCurve=DataInterface() {propertyValues {Curve {
            defaultValue=1
            keys [
             {arriveTangent=18.6243, interpMode=RCIM_Cubic, leaveTangent=18.6243, tangentMode=RCTM_User, time=0, value=0}
             {arriveTangent=-11.3665, interpMode=RCIM_Cubic, leaveTangent=-11.3665, tangentMode=RCTM_User, time=0.1, value=4}
             {arriveTangent=0.0224, interpMode=RCIM_Cubic, leaveTangent=0.0224, tangentMode=RCTM_User, time=1, value=-2}
            ]
           }}})
        Bias=0
       )
      )
      Destination Space=::Simulation
      Apply Scale=true
     )
    )) {enabled=false, isSetParameters=true}
   &ScaleColor(
    Scale Alpha=&FloatFromCurve(FloatCurve=DataInterface() {propertyValues {Curve {keys [<at 0 1>,<at 0.2 0.4>,<at 1 0.3>]}}})
    ScaleA=true
    ScaleRGB=true
   )
   /SetVariables(Particles.Position=&AddVectorToPosition(
     Position=$Particles.Position
     Vector=&TransformVector(
      Vector=&MakeVector(
       X=0
       Y=&Cosine(
        Normalized Angle=&Add_Float(
         A=&Multiply_Float(A=$Particles.NormalizedAge, B=2)
         B=&Multiply_Float(A=$Emitter.LoopedAge, B=-2)
        )
        Period=1
        Scale=1
        Bias=0
       )
       Z=&Sine(
        Normalized Angle=&Add_Float(
         A=&Multiply_Float(A=$Particles.NormalizedAge, B=2)
         B=&Multiply_Float(A=$Emitter.LoopedAge, B=-2)
        )
        Period=1
        Scale=&FloatFromCurve(
         FloatCurve=DataInterface() {propertyValues {Curve {
            defaultValue=1
            keys [
             {arriveTangent=18.6243, interpMode=RCIM_Cubic, leaveTangent=18.6243, tangentMode=RCTM_User, time=0, value=0}
             {arriveTangent=-11.3665, interpMode=RCIM_Cubic, leaveTangent=-11.3665, tangentMode=RCTM_User, time=0.1, value=5}
             {arriveTangent=0.0224, interpMode=RCIM_Cubic, leaveTangent=0.0224, tangentMode=RCTM_User, time=1, value=-2}
            ]
           }}}
         Scale Curve=0.3
        )
        Bias=0
       )
      )
      Destination Space=::Simulation
      Apply Scale=true
     )
    )) {enabled=false, isSetParameters=true}
   /SetVariables(Particles.Position=&AddVectorToPosition(
     Position=$Particles.Position
     Vector=&TransformVector(
      Vector=&MakeVector(
       X=0
       Y=&Cosine(
        Normalized Angle=&Add_Float(
         A=&Multiply_Float(A=$Particles.NormalizedAge, B=1)
         B=&Multiply_Float(A=$Emitter.LoopedAge, B=-2)
        )
        Period=1
        Scale=&FloatFromCurve(FloatCurve=DataInterface() {propertyValues {Curve {
            defaultValue=1
            keys [
             {arriveTangent=18.6243, interpMode=RCIM_Cubic, leaveTangent=18.6243, tangentMode=RCTM_User, time=0, value=0}
             {arriveTangent=-11.3665, interpMode=RCIM_Cubic, leaveTangent=-11.3665, tangentMode=RCTM_User, time=0.05, value=3}
             {arriveTangent=0.0224, interpMode=RCIM_Cubic, leaveTangent=0.0224, tangentMode=RCTM_User, time=0.3, value=-1}
            ]
           }}})
        Bias=0
       )
       Z=&Sine(
        Normalized Angle=&Add_Float(
         A=&Multiply_Float(A=$Particles.NormalizedAge, B=1)
         B=&Multiply_Float(A=$Emitter.LoopedAge, B=-2)
        )
        Period=1
        Scale=&FloatFromCurve(
         FloatCurve=DataInterface() {propertyValues {Curve {
            defaultValue=1
            keys [
             {arriveTangent=18.6243, interpMode=RCIM_Cubic, leaveTangent=18.6243, tangentMode=RCTM_User, time=0, value=0}
             {arriveTangent=-11.3665, interpMode=RCIM_Cubic, leaveTangent=-11.3665, tangentMode=RCTM_User, time=0.05, value=3}
             {arriveTangent=0.0224, interpMode=RCIM_Cubic, leaveTangent=0.0224, tangentMode=RCTM_User, time=0.3, value=-1}
            ]
           }}}
         Scale Curve=0.5
        )
        Bias=0
       )
      )
      Destination Space=::Simulation
      Apply Scale=true
     )
    )) {isSetParameters=true}
   &VC_UpdateRibbon_v1(Enable Skeletal Position=true)
  ]
  rendererClasses [&NiagaraRibbonRendererProperties]
  renderers [
   {rendererClass=&NiagaraRibbonRendererProperties}
   {rendererClass=&NiagaraRibbonRendererProperties, rendererIndex=1}
  ]
 }
}
summary {
 systemName=NS_Boost_Holidaydeco
 emitters [
  {
   emitterName=NE_Cristmasballs
   rendererClasses [&NiagaraMeshRendererProperties, &NiagaraSpriteRendererProperties]
  }
  {emitterName=NE_Hollyleaves, rendererClasses [&NiagaraSpriteRendererProperties]}
  {
   bEnabled=false
   emitterName=NE_Cristmaslights
   rendererClasses [&NiagaraMeshRendererProperties, &NiagaraSpriteRendererProperties]
  }
  {bEnabled=false, emitterName=NE_White_Trail001, rendererClasses [&NiagaraRibbonRendererProperties]}
  {emitterName=NE_Green_Trail, rendererClasses [&NiagaraRibbonRendererProperties]}
  {emitterName=NE_Red_Trail002, rendererClasses [&NiagaraRibbonRendererProperties]}
  {emitterName=NE_White_Trail, rendererClasses [&NiagaraRibbonRendererProperties]}
 ]
 userVariables [
  User.bIsFrontendPreview {type=NiagaraBool}
  User.bIsPainted {type=NiagaraBool}
  User.blsIsCoroColor {type=NiagaraBool}
  User.bPreviewVehicleSpeed {type=NiagaraBool}
  User.ColorVariantID {type=NiagaraInt32}
  User.DTCarOffset {type=Vector3f, description="", defaultValue {struct=Vector3f, value=<0 0 5>}}
  User.GameModeID {type=NiagaraInt32, defaultValue=3}
  User.PaintedColor {type=LinearColor, description="", defaultValue {struct=LinearColor, value=<rgba 1 1 1 1>}}
  User.PaintedScalar {type=NiagaraFloat, defaultValue=1, description=""}
  User.PortCount {type=NiagaraInt32, defaultValue=2}
 ]
}
